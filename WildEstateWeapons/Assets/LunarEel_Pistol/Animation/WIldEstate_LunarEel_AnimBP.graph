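# Animation Blueprint for the LunarEel pistol skeleton in WildEstateWeapons, driving the weapon mesh's pose.
# A non-looping sequence player of the exotic Fired animation feeds DefaultSlot, which outputs the final pose and lets montages override it.
# It has no state machine, variables or event logic, so the Fired sequence is the base pose source rather than being triggered by this graph.
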AnimBlueprint /WildEstateWeapons/Assets/LunarEel_Pistol/Animation/WIldEstate_LunarEel_AnimBP
target_skeleton=/WildEstateWeapons/Assets/LunarEel_Pistol/Meshes/LunarEel_Pistol_Skeleton
graphs [
 AnimGraph {nodes [Output Pose(Result=Slot 'DefaultSlot'(Source=Sequence Player 'WildEstate_LunarEel_Pistol_Fired_Exotic_WEP'() {bLoopAnimation=false, sequence=&/WildEstate_LunarEel_Pistol_Fired_Exotic_WEP}))]}
 EventGraph {links [], nodes []}
]
variables []
