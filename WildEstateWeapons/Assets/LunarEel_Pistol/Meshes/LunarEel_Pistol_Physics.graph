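# Physics asset for the LunarEel pistol mesh, covering a balloon on a string, apparently a dangling decoration on the weapon.
# Four small capsules form the string and a much larger zero-length capsule, effectively a sphere, forms the balloon.
# Constraints link them in a chain from string_01 to balloon_base, all sharing the same swing and twist limit angles.
# No other pistol bones have bodies.
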PhysicsAsset /WildEstateWeapons/Assets/LunarEel_Pistol/Meshes/LunarEel_Pistol_Physics
bodies {
 balloon_base [{shapeType=Capsule, length=0, radius=14, shapeName=balloon_base_capsule, center=<1.0571 -0.2901 13.9697>, rotation=<rot 0 -90 0>}]
 string_01 [{shapeType=Capsule, length=1.505, radius=0.505, shapeName=string_01_capsule, center=<1.1633 0.0039 0.001>, rotation=<rot 0 -90.2265 -89.9393>}]
 string_02 [{shapeType=Capsule, length=1.505, radius=0.505, shapeName=string_02_capsule, center=<1.0541 0 0>, rotation=<rot 0 90 -90>}]
 string_03 [{shapeType=Capsule, length=1.005, radius=0.505, shapeName=string_03_capsule, center=<1.0708 0 0>, rotation=<rot 0 90 -90>}]
 string_04 [{shapeType=Capsule, length=0.505, radius=0.505, shapeName=string_04_capsule, center=<-6.6202 0 0>, rotation=<rot -15.3477 -89.9999 -90.0004>}]
}
constraints [
 {bone1Name=string_02, bone2Name=string_01, swing1LimitDegrees=45, swing2LimitDegrees=45, twistLimitDegrees=30}
 {bone1Name=string_03, bone2Name=string_02, swing1LimitDegrees=45, swing2LimitDegrees=45, twistLimitDegrees=30}
 {bone1Name=string_04, bone2Name=string_03, swing1LimitDegrees=45, swing2LimitDegrees=45, twistLimitDegrees=30}
 {bone1Name=balloon_base, bone2Name=string_04, swing1LimitDegrees=45, swing2LimitDegrees=45, twistLimitDegrees=30}
]
