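# Animation Blueprint for the Lunar Eel exotic pistol, posing the weapon and simulating its string and balloon.
# An inherited bFired flag switches instantly between fired and loaded animations, feeding a RigidBody node with upward gravity for balloon-like float.
# Bone modifiers then shrink string_01 and counter-scale balloon_base before the default montage slot.
# Each update resets dynamics when the current weapon mesh turns visible, avoiding stale physics.
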AnimBlueprint /WildEstateWeapons/Assets/LunarEel_Pistol/Animation/WIldEstate_LunarEel_Exotic_AnimBP
target_skeleton=/WildEstateWeapons/Assets/LunarEel_Pistol/Meshes/LunarEel_Pistol_Skeleton
graphs [
 AnimGraph {nodes [
   n1 = Sequence Player 'WildEstate_LunarEel_Pistol_Fired_Exotic_WEP'() {bLoopAnimation=false, sequence=&/WildEstate_LunarEel_Pistol_Fired_Exotic_WEP}
   n2 = Sequence Player 'WildEstate_LunarEel_Pistol_Loaded_Exotic_WEP'() {bLoopAnimation=false, sequence=&/WildEstate_LunarEel_Pistol_Loaded_Exotic_WEP}
   n3 = Get bFired() -> (bFired:bool)
   n4 = RigidBody!(
    ComponentPose=Local To Component(LocalPose=Blend Poses by bool(BlendPose_0=@n1, BlendPose_1=@n2, BlendTime_0=0, BlendTime_1=0, bActiveValue=@n3))
    bAlphaBoolEnabled:bool
    ExternalForce
   ) -> (Pose) {
    alphaInputType=Bool
    bForceDisableCollisionBetweenConstraintBodies=true
    bOverrideWorldGravity=true
    overridePhysicsAsset=/WildEstateWeapons/Assets/LunarEel_Pistol/Meshes/LunarEel_Pistol_Physics
    baseBoneRef {boneName=string_01}
    overrideWorldGravity {z=400}
    simSpaceSettings {maxAngularVelocity=12, maxLinearVelocity=800, worldAlpha=1}
   }
   n5 = "Transform (Modify) Bone - Bone: string_01"!(ComponentPose=@n4, bAlphaBoolEnabled:bool=True, Translation="0,0,0", Rotation="0,0,0", Scale="0.25,0.25,0.25") -> (Pose) {alphaInputType=Bool, scaleMode=BMM_Replace, scaleSpace=BCS_BoneSpace, boneToModify {boneName=string_01}, scale=<0.25 0.25 0.25>}
   n6 = "Transform (Modify) Bone - Bone: balloon_base"!(ComponentPose=@n5, bAlphaBoolEnabled:bool=True, Translation="0,0,0", Rotation="0,0,0", Scale="4,4,4") -> (Pose) {alphaInputType=Bool, scaleMode=BMM_Replace, scaleSpace=BCS_BoneSpace, boneToModify {boneName=balloon_base}, scale=<4 4 4>}
   Output Pose(Result=Slot 'DefaultSlot'(Source=Component To Local(ComponentPose=@n6)))
  ]}
 EventGraph {nodes [
   n1 = Event BlueprintUpdateAnimation()
   > Parent: BlueprintUpdateAnimation!(execute, DeltaTimeX:float=@n1.DeltaTimeX) -> (then) {is=title K2Node_CallParentFunction}
   n2 = Get WeaponMesh!(self:Fort Weapon Object Reference=GetCurrentWeapon!(self:Fort Pawn Object Reference=TryGetFortPawn()) -> (ReturnValue:Fort Weapon Object Reference)) -> (WeaponMesh:Skeletal Mesh Component Budgeted Object Reference)
   n3 = IsVisible(self:Scene Component Object Reference=@n2)
   > n4 = Branch(Condition=@n3)
   > n5 = Sequence()
   > n6 = Branch(Condition=Get WasVisible() -> (WasVisible:bool))
   ResetAnimInstanceDynamics!(execute=@n6.else, self:Skeletal Mesh Component Object Reference=@n2, InTeleportType:ETeleportType Enum=ResetPhysics) -> (then)
   Set WasVisible(execute=[@n5.then_1, @n4.else], WasVisible:bool=@n3) -> (Output_Get:bool)
  ]}
]
variables [
 ShouldResetDynamics {type=boolean}
 WasVisible {type=boolean}
]
