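# Test animation Blueprint for the LunarEel pistol that exercises string dynamics on top of its loaded pose.
# Each update, two offset cosine waves of time sway the root bone in world-space X and Y, feeding motion into a RigidBody node with upward gravity override.
# Alpha blends the simulation and, through a clamped cube-root curve, shrinks the string_01 bone toward quarter scale afterward.
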AnimBlueprint /WildEstateWeapons/Assets/LunarEel_Pistol/Meshes/ABP_LunarEel_Pistol_DynTest
target_skeleton=&/LunarEel_Pistol_Skeleton
graphs [
 AnimGraph {nodes [
   n1 = Sequence Player 'LunarEel_Pistol_Loaded_Pose'() {bLoopAnimation=false, sequence=&/LunarEel_Pistol_Loaded_Pose}
   n2 = RigidBody(
    ComponentPose="Transform (Modify) Bone - Bone: root"!(
     ComponentPose=Local To Component(LocalPose=@n1)
     Alpha
     Translation_X:double=Get NewVar_0() -> (NewVar_0:double)
     Translation_Y:double=Get NewVar() -> (NewVar:double)
     Translation_Z:double=0
     Rotation="0,0,0"
     Scale
    ) -> (Pose) {translationMode=BMM_Replace, translationSpace=BCS_WorldSpace, boneToModify {boneName=root}, translation {y=15}}
    Alpha=Get Alpha() -> (Alpha:float)
   ) {
    bForceDisableCollisionBetweenConstraintBodies=true
    bOverrideWorldGravity=true
    overridePhysicsAsset=&/LunarEel_Pistol_Physics
    baseBoneRef {boneName=string_01}
    overrideWorldGravity {z=400}
   }
   n3 = Get Scale() -> (Scale:Vector)
   Output Pose(Result=Component To Local(ComponentPose="Transform (Modify) Bone - Bone: string_01"(ComponentPose=@n2, Alpha=1, Translation="0,0,0", Rotation="0,0,0", Scale=@n3) {scaleMode=BMM_Replace, scaleSpace=BCS_BoneSpace, boneToModify {boneName=string_01}}))
  ]}
 EventGraph {nodes [
   Event BlueprintUpdateAnimation()
   n1 = GetTimeSeconds!() -> (ReturnValue:double)
   n2 = float * float(A=@n1, B=220) {is=title K2Node_PromotableOperator}
   > Set NewVar(NewVar:double=float * float(A="Cos (Degrees)"!(A:double=@n2) -> (ReturnValue:double), B=15)) -> (Output_Get:double)
   n3 = float + float(A=float * float(A=@n1, B=221), B=100)
   > Set NewVar_0(NewVar_0:double=float * float(A="Cos (Degrees)"(A=@n3), B=15)) -> (Output_Get:double)
   n4 = Power!(Base:double=Get Alpha(), Exp:double=0.3333) -> (ReturnValue:double) {is=title K2Node_CallFunction}
   > Set Scale(Scale:Vector="Lerp (Vector)"!(A:Vector="1,1,1", B:Vector="0.25,0.25,0.25", Alpha:float="Clamp (Float)"(Value=@n4, Min=0, Max=1)) -> (ReturnValue:Vector)) -> (Output_Get:Vector)
   TryGetPawnOwner()
  ]}
]
variables [
 Alpha {type=number, default=1}
 NewVar {type=number}
 NewVar_0 {type=number}
 Scale {type=Vector, default=<1 1 1>}
]
